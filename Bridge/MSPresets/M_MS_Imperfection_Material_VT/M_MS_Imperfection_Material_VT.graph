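# Bridge preset Default Lit material for imperfection surfaces, built on a tiled virtual-texture MR map sampled through MF_Tiling.
# The map's green channel remaps between min and max roughness, then blends against a constant roughness.
# A static switch optionally lets the red channel mask that value, weighted by Mask Influence.
# The result blends two albedo colors, alongside scalar Metallic and Specular parameters.
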Material /Bridge/MSPresets/M_MS_Imperfection_Material_VT/M_MS_Imperfection_Material_VT
graph {
 properties {pixelDepthOffsetMode=PDOM_Legacy, usedShadingModels=Default Lit}
 nodes [
  n1 = TextureSampleParameter2D(UVs=MaterialFunctionCall() {materialFunction=/Bridge/MSPresets/M_MS_Default_Material/Functions/MF_Tiling}.Result) {
   group=00 - Global
   parameterName=MR
   samplerSource=SSM_Wrap_WorldGroupSettings
   samplerType=SAMPLERTYPE_VirtualLinearColor
   sortPriority=0
   texture=/Bridge/MSPresets/MSVTTextures/WhitePlaceholder
  }
  n2 = LinearInterpolate(
   A=ScalarParameter() {defaultValue=0.5, group=02 - Properties, parameterName=Roughness Constant, sortPriority=3}
   B=LinearInterpolate(
    A=ScalarParameter() {group=02 - Properties, parameterName=Min Roughness Mask, sortPriority=5}
    B=ScalarParameter() {defaultValue=1, group=02 - Properties, parameterName=Max Roughness Mask, sortPriority=6}
    Alpha=@n1.G
   )
   Alpha=ScalarParameter() {defaultValue=1, group=02 - Properties, parameterName=Use Roughness Mask?, sortPriority=4}
  )
  n3 = LinearInterpolate(B=@n2, Alpha=@n1.R)
  n4 = ScalarParameter() {group=02 - Properties, parameterName=Mask Influence?, sortPriority=7}
  LinearInterpolate(
   A=VectorParameter() {group=01 - Albedo, parameterName=Color A, sortPriority=0, defaultValue=<rgba 0 0 0 1>}.RGB
   B=VectorParameter() {group=01 - Albedo, parameterName=Color B, sortPriority=1, defaultValue=<rgba 0.5 0.5 0.5 1>}.RGB
   Alpha=StaticSwitchParameter(True=LinearInterpolate(A=@n2, B=@n3, Alpha=@n4), False=@n2) {group=02 - Properties, parameterName=Use Mask Texture Channel?, sortPriority=0}
  )
  ScalarParameter() {defaultValue=0.5, group=02 - Properties, parameterName=Specular, sortPriority=2}
  ScalarParameter() {defaultValue=1, group=02 - Properties, parameterName=Metallic, sortPriority=1}
 ]
}
